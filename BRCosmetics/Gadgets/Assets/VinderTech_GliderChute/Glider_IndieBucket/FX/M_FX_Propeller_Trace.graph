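# Translucent, two-sided Niagara sprite material for the propeller trace effect on the IndieBucket glider.
# Color comes from gradient ramps over a polar UV channel, with scalar parameters selecting among color variants.
# The secondary variant adds a reflected gradient on one color pair and a view-angle fade against AngleFadeColor.
# Particle color tints the result before emissive adjustments, and particle alpha scales the trace texture's red channel as a mask.
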Material /BRCosmetics/Gadgets/Assets/VinderTech_GliderChute/Glider_IndieBucket/FX/M_FX_Propeller_Trace
graph {
 properties {bUsedWithNiagaraSprites=true, blendMode=BLEND_Translucent, pixelDepthOffsetMode=PDOM_Legacy, twoSided=true, usedShadingModels=Default Lit}
 nodes [
  n1 = ComponentMask(None=MaterialFunctionCall!(UV, Twist Center, Twist Edges) {materialFunction=/Game/Accessories/FORT_Backpacks/Backpack_F_MED_Celestial_Glow/FX/Materials/MF_PolarUV}.Polar UVs) {g=true}
  n2 = SmoothStep(Value=@n1) {constMin=0.4}
  n3 = LinearInterpolate(
   A=LinearInterpolate(
    A=VectorParameter() {parameterName=VariantColor1A, defaultValue=<rgba 0 1.45 1.5 1>}.RGB
    B=VectorParameter() {parameterName=VariantColor1B, defaultValue=<rgba 0.521 0.9647 0.1946 1>}.RGB
    Alpha=@n2
   )
   B=LinearInterpolate(
    A=VectorParameter() {parameterName=VariantColor2A, defaultValue=<rgba 0.1887 0.09 0.7379 1>}.RGB
    B=VectorParameter() {parameterName=VariantColor2B, defaultValue=<rgba 0.1998 0.2979 0.7991 1>}.RGB
    Alpha=@n2
   )
   Alpha=ScalarParameter() {parameterName=Switch Color}
  )
  n4 = VectorParameter() {parameterName=VariantColor1C, defaultValue=<rgba 0.7454 0 0.5972 1>}
  n5 = VectorParameter() {parameterName=VariantColor1D, defaultValue=<rgba 0.5457 0.4125 0.8388 1>}
  n6 = MaterialFunctionCall!(Gradient=@n1, Center Point=ScalarParameter() {defaultValue=0.7}) {materialFunction=/Game/Effects/Fort_Effects/MaterialFunctions/MF_GradientReflected}
  n7 = VectorParameter() {parameterName=VariantColor2C, defaultValue=<rgba 0.573 0.0286 0.6105 1>}
  n8 = VectorParameter() {parameterName=VariantColor2D, defaultValue=<rgba 0.5457 0.4125 0.8388 1>}
  n9 = LinearInterpolate(
   A=VectorParameter() {parameterName=AngleFadeColor, defaultValue=<rgba 0.288 0.5405 0.9216 1>}.RGB
   B=LinearInterpolate(
    A=LinearInterpolate(A=@n4.RGB, B=@n5.RGB, Alpha=@n6.Result)
    B=LinearInterpolate(A=@n7.RGB, B=@n8.RGB, Alpha=@n2)
    Alpha=ScalarParameter() {parameterName=Switch Color}
   )
   Alpha=MaterialFunctionCall("FadeSensitivity (optional)"=ScalarParameter() {defaultValue=1, parameterName=FadeAngleStrenght}) {materialFunction=&FadeBasedOnViewAngle}.Result
  )
  n10 = ScalarParameter() {parameterName=Use Sec Variant?}
  n11 = ParticleColor()
  MaterialFunctionCall(Emissive=Multiply(A=LinearInterpolate(A=@n3, B=@n9, Alpha=@n10), B=@n11.RGB)) {materialFunction=&MF_FX_EmissiveAdjustments}
  Multiply(
   A=TextureSample() {samplerType=SAMPLERTYPE_LinearColor, texture=/Game/Gadgets/Assets/VinderTech_GliderChute/Glider_IndieBucket/FX/T_FX_Propeller_Trace}.R
   B=@n11.A
  )
 ]
}
